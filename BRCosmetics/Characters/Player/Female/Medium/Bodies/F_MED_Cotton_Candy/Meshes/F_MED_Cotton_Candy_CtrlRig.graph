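# Control Rig for the female Cotton Candy body cosmetic that adjusts the jacket bones as the torso moves.
# A spherical pose reader on spine_04, relative to spine_01, weights small offsets to the side and back jacket base bones.
# When NoAnimDynamics is true, clamped distances between body and jacket points weight rotations of the side, back and front base bones, approximating collision without physics.
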ControlRigBlueprint /BRCosmetics/Characters/Player/Female/Medium/Bodies/F_MED_Cotton_Candy/Meshes/F_MED_Cotton_Candy_CtrlRig
graphs [
 RigVMFunctionLibrary {links [], nodes []}
 RigVMModel {nodes [
   RigUnit_BeginExecution = &RigUnit_BeginExecution()
   n1 = &RigUnit_SphericalPoseReader(
    ActiveRegionScaleFactors=(PositiveWidth=1,NegativeWidth=1,PositiveHeight=1,NegativeHeight=1)
    ActiveRegionSize=0.1
    Debug=(bDrawDebug=false,DebugScale=25,DebugSegments=20,DebugThickness=0.25)
    DriverAxis=(X=-1,Y=0,Z=0)
    DriverItem=(Type=Bone,Name="spine_04")
    FalloffRegionScaleFactors=(PositiveWidth=1,NegativeWidth=1,PositiveHeight=1,NegativeHeight=1)
    FalloffSize=0.2
    OptionalParentItem=(Type=Bone,Name="spine_01")
    RotationOffset=(X=90,Y=0,Z=150)
    ExecutePin=@RigUnit_BeginExecution.ExecutePin
   )
   > &RigUnit_ModifyTransforms(
    ItemToModify=((Item=(Type=Bone,Name="dyn_jacket_side_a_base_r"),Transform=(Rotation=(X=0,Y=0,Z=0,W=1),Translation=(X=0,Y=1.4,Z=0),Scale3D=(X=1,Y=1,Z=1))))
    Weight=@n1.OutputParam
   )
   > &RigUnit_ModifyTransforms(
    ItemToModify=((Item=(Type=Bone,Name="dyn_jacket_side_a_base_l"),Transform=(Rotation=(X=0,Y=0,Z=0,W=1),Translation=(X=0,Y=-1.4,Z=0),Scale3D=(X=1,Y=1,Z=1))))
    Weight=@n1.OutputParam
   )
   > n2 = &RigUnit_ModifyTransforms(
    ItemToModify=((Item=(Type=Bone,Name="dyn_jacket_bk_a_base"),Transform=(Rotation=(X=0,Y=0,Z=0,W=1),Translation=(X=0,Y=-1,Z=0),Scale3D=(X=1,Y=1,Z=1))))
    Weight=@n1.OutputParam
   )
   n3 = &RigVMFunction_ControlFlowBranch(ExecuteContext=@n2.ExecutePin, Condition=RigVMVariableNode() {variable=NoAnimDynamics}.Value)
   n4 = &RigUnit_GetTransform(Item=(Type=Control,Name="Side body R"))
   n5 = &RigUnit_GetTransform(Item=(Type=Bone,Name="dyn_jacket_side_b_r"))
   &RigUnit_ModifyTransforms(
    ItemToModify=((Item=(Type=Bone,Name="dyn_jacket_side_a_base_r"),Transform=(Rotation=(X=0,Y=0,Z=0,W=1),Translation=(X=0,Y=0,Z=0),Scale3D=(X=1,Y=1,Z=1))))
    ExecutePin=@n3.True
    Weight=&RigVMFunction_MathFloatRemap(
     SourceMaximum=7
     SourceMinimum=0
     TargetMaximum=0
     TargetMinimum=1
     bClamp=true
     Value=&RigVMFunction_MathVectorDistance(B=@n4.Transform.Translation, A=@n5.Transform.Translation).Result
    ).Result
    ItemToModify.0.Transform.Rotation=&RigVMFunction_MathQuaternionFromEuler(Euler=(X=0,Y=0,Z=-40)).Result
   )
   n6 = &RigUnit_GetTransform(Item=(Type=Control,Name="back body"))
   n7 = &RigUnit_GetTransform(Item=(Type=Bone,Name="dyn_jacket_bk_b"))
   > &RigUnit_ModifyTransforms(
    ItemToModify=((Item=(Type=Bone,Name="dyn_jacket_bk_a_base"),Transform=(Rotation=(X=0,Y=0,Z=0,W=1),Translation=(X=0,Y=0,Z=0),Scale3D=(X=1,Y=1,Z=1))))
    Weight=&RigVMFunction_MathFloatRemap(
     SourceMaximum=3
     SourceMinimum=0
     TargetMaximum=0
     TargetMinimum=1
     bClamp=true
     Value=&RigVMFunction_MathVectorDistance(B=@n6.Transform.Translation, A=@n7.Transform.Translation).Result
    ).Result
    ItemToModify.0.Transform.Rotation=&RigVMFunction_MathQuaternionFromEuler(Euler=(X=0,Y=0,Z=-60)).Result
   )
   n8 = &RigUnit_GetTransform(Item=(Type=Control,Name="Side body L"))
   n9 = &RigUnit_GetTransform(Item=(Type=Bone,Name="dyn_jacket_side_b_l"))
   > &RigUnit_ModifyTransforms(
    ItemToModify=((Item=(Type=Bone,Name="dyn_jacket_side_a_base_l"),Transform=(Rotation=(X=0,Y=0,Z=0,W=1),Translation=(X=0,Y=0,Z=0),Scale3D=(X=1,Y=1,Z=1))))
    Weight=&RigVMFunction_MathFloatRemap(
     SourceMaximum=5
     SourceMinimum=0
     TargetMaximum=0
     TargetMinimum=1
     bClamp=true
     Value=&RigVMFunction_MathVectorDistance(B=@n8.Transform.Translation, A=@n9.Transform.Translation).Result
    ).Result
    ItemToModify.0.Transform.Rotation=&RigVMFunction_MathQuaternionFromEuler(Euler=(X=0,Y=0,Z=-30)).Result
   )
   n10 = &RigUnit_GetTransform(Item=(Type=Control,Name="Side body L_2"))
   n11 = &RigUnit_GetTransform(Item=(Type=Bone,Name="dyn_jacket_side_b_l"))
   > &RigUnit_ModifyTransforms(
    ItemToModify=((Item=(Type=Bone,Name="dyn_jacket_side_a_base_l"),Transform=(Rotation=(X=0,Y=0,Z=0,W=1),Translation=(X=0,Y=0,Z=0),Scale3D=(X=1,Y=1,Z=1))))
    Weight=&RigVMFunction_MathFloatRemap(
     SourceMaximum=5
     SourceMinimum=0
     TargetMaximum=0
     TargetMinimum=1
     bClamp=true
     Value=&RigVMFunction_MathVectorDistance(B=@n10.Transform.Translation, A=@n11.Transform.Translation).Result
    ).Result
    ItemToModify.0.Transform.Rotation=&RigVMFunction_MathQuaternionFromEuler(Euler=(X=0,Y=0,Z=-30)).Result
   )
   n12 = &RigUnit_GetTransform(Item=(Type=Control,Name="body FL"))
   n13 = &RigUnit_GetTransform(Item=(Type=Control,Name="jacket BL"))
   > &RigUnit_ModifyTransforms(
    ItemToModify=((Item=(Type=Bone,Name="dyn_jacket_side_a_base_l"),Transform=(Rotation=(X=0,Y=0,Z=0,W=1),Translation=(X=0,Y=0,Z=0),Scale3D=(X=1,Y=1,Z=1))))
    Weight=&RigVMFunction_MathFloatRemap(
     SourceMaximum=5
     SourceMinimum=0
     TargetMaximum=0
     TargetMinimum=1
     bClamp=true
     Value=&RigVMFunction_MathVectorDistance(B=@n12.Transform.Translation, A=@n13.Transform.Translation).Result
    ).Result
    ItemToModify.0.Transform.Rotation=&RigVMFunction_MathQuaternionFromEuler(Euler=(X=0,Y=0,Z=-60)).Result
   )
   n14 = &RigUnit_GetTransform(Item=(Type=Control,Name="body FL"))
   n15 = &RigUnit_GetTransform(Item=(Type=Control,Name="Jacket FL "))
   > &RigUnit_ModifyTransforms(
    ItemToModify=((Item=(Type=Bone,Name="dyn_jacket_fr_a_base_l"),Transform=(Rotation=(X=0,Y=0,Z=0,W=1),Translation=(X=0,Y=0,Z=0),Scale3D=(X=1,Y=1,Z=1))))
    Weight=&RigVMFunction_MathFloatRemap(
     SourceMaximum=5
     SourceMinimum=0
     TargetMaximum=0
     TargetMinimum=1
     bClamp=true
     Value=&RigVMFunction_MathVectorDistance(B=@n14.Transform.Translation, A=@n15.Transform.Translation).Result
    ).Result
    ItemToModify.0.Transform.Rotation=&RigVMFunction_MathQuaternionFromEuler(Euler=(X=0,Y=0,Z=-40)).Result
   )
   n16 = &RigUnit_GetTransform(Item=(Type=Control,Name="Jacket FL "))
   n17 = &RigUnit_GetTransform(Item=(Type=Bone,Name="spine_01"))
   > &RigUnit_ModifyTransforms(
    ItemToModify=((Item=(Type=Bone,Name="dyn_jacket_fr_a_base_l"),Transform=(Rotation=(X=0,Y=0,Z=0,W=1),Translation=(X=0,Y=0,Z=0),Scale3D=(X=1,Y=1,Z=1))))
    Weight=&RigVMFunction_MathFloatRemap(
     SourceMaximum=15
     SourceMinimum=5
     TargetMaximum=0
     TargetMinimum=1
     bClamp=true
     Value=&RigVMFunction_MathVectorDistance(B=@n16.Transform.Translation, A=@n17.Transform.Translation).Result
    ).Result
    ItemToModify.0.Transform.Rotation=&RigVMFunction_MathQuaternionFromEuler(Euler=(X=0,Y=0,Z=-50)).Result
   )
   n18 = &RigUnit_GetTransform(Item=(Type=Control,Name="body FL"))
   n19 = &RigUnit_GetTransform(Item=(Type=Control,Name="Jacket FL "))
   > &RigUnit_ModifyTransforms(
    ItemToModify=((Item=(Type=Bone,Name="dyn_jacket_fr_a_base_l"),Transform=(Rotation=(X=0,Y=0,Z=0,W=1),Translation=(X=0,Y=0,Z=0),Scale3D=(X=1,Y=1,Z=1))))
    Weight=&RigVMFunction_MathFloatRemap(
     SourceMaximum=5
     SourceMinimum=0
     TargetMaximum=0
     TargetMinimum=1
     bClamp=true
     Value=&RigVMFunction_MathVectorDistance(B=@n18.Transform.Translation, A=@n19.Transform.Translation).Result
    ).Result
    ItemToModify.0.Transform.Rotation=&RigVMFunction_MathQuaternionFromEuler(Euler=(X=0,Y=0,Z=-40)).Result
   )
   n20 = &RigUnit_GetTransform(Item=(Type=Control,Name="jacket FR"))
   n21 = &RigUnit_GetTransform(Item=(Type=Bone,Name="spine_01"))
   > &RigUnit_ModifyTransforms(
    ItemToModify=((Item=(Type=Bone,Name="dyn_jacket_fr_a_base_r"),Transform=(Rotation=(X=0,Y=0,Z=0,W=1),Translation=(X=0,Y=0,Z=0),Scale3D=(X=1,Y=1,Z=1))))
    Weight=&RigVMFunction_MathFloatRemap(
     SourceMaximum=10
     SourceMinimum=5
     TargetMaximum=0
     TargetMinimum=1
     bClamp=true
     Value=&RigVMFunction_MathVectorDistance(B=@n20.Transform.Translation, A=@n21.Transform.Translation).Result
    ).Result
    ItemToModify.0.Transform.Rotation=&RigVMFunction_MathQuaternionFromEuler(Euler=(X=0,Y=0,Z=-40)).Result
   )
   &RigVMFunction_MathQuaternionFromEuler(Euler=(X=0,Y=20,Z=-15))
   &RigVMFunction_MathQuaternionFromEuler(Euler=(X=0,Y=20,Z=-15))
   &RigVMFunction_MathQuaternionFromEuler(Euler=(X=0,Y=0,Z=-15))
  ]}
]
hierarchy [
 root {loc=<0 0 0>, rot=<0 0 0 1>}
 pelvis {parent=root, loc=<0 -0.5301 98.0571>, rot=<0.0127 0.707 -0.0127 -0.707>}
 spine_01 {parent=pelvis, loc=<6.1208 -0.9328 0>, rot=<0 0 0.034 -0.9994>}
 spine_02 {parent=spine_01, loc=<6.5996 -0.2095 0>, rot=<0 0 0.0238 -0.9997>}
 spine_03 {parent=spine_02, loc=<6.4914 -0.1237 0>, rot=<0 0 0.0149 -0.9999>}
 spine_04 {parent=spine_03, loc=<6.9598 0.0794 0>, rot=<0 0 -0.0132 -0.9999>}
 spine_05 {parent=spine_04, loc=<6.9987 0 0>, rot=<0 0 -0.0015 -1>}
 clavicle_l {parent=spine_05, loc=<5.4831 -0.1053 -3.2946>, rot=<0.0465 -0.7289 -0.009 -0.6829>}
 upperarm_l {parent=clavicle_l, loc=<9.8658 -1.2136 -0.4105>, rot=<0.0164 -0.3245 -0.0269 -0.9454>}
 lowerarm_l {parent=upperarm_l, loc=<26.465 -0.8345 0.259>, rot=<0.0036 0.0161 -0.2298 -0.9731>}
 hand_l {parent=lowerarm_l, loc=<22.7355 -0.3203 -0.6105>, rot=<-0.6784 -0.0223 0.0735 -0.7307>}
 index_metacarpal_l {parent=hand_l, loc=<1.6535 0.0183 -2.96>, rot=<0.0158 -0.0124 -0.0173 -0.9996>}
 index_01_l {parent=index_metacarpal_l, loc=<8.3008 0.0357 -0.0524>, rot=<-0.0044 0.0485 0.0924 -0.9945>}
 index_02_l {parent=index_01_l, loc=<3.3587 -0.3004 0.0452>, rot=<-0.002 0.0027 0.3225 -0.9466>}
 index_03_l {parent=index_02_l, loc=<2.5578 -0.123 -0.0252>, rot=<-0.002 0.0067 0.2558 -0.9667>}
 middle_metacarpal_l {parent=hand_l, loc=<2.1856 -0.1725 -0.5984>, rot=<0.006 0.0168 -0.0329 -0.9993>}
 middle_01_l {parent=middle_metacarpal_l, loc=<7.3275 -0.2447 0.0129>, rot=<-0.0208 0.0042 0.0858 -0.9961>}
 middle_02_l {parent=middle_01_l, loc=<4.0123 0.1401 0.1175>, rot=<-0.0218 0.0258 0.3512 -0.9357>}
 middle_03_l {parent=middle_02_l, loc=<2.6077 0.0534 0.1373>, rot=<0.0082 0.0188 0.1708 -0.9851>}
 pinky_metacarpal_l {parent=hand_l, loc=<1.7616 -0.9753 3.3814>, rot=<-0.0397 0.0814 -0.069 -0.9935>}
 pinky_01_l {parent=pinky_metacarpal_l, loc=<7.1709 -0.4831 -0.1094>, rot=<-0.0654 -0.0035 0.2242 -0.9723>}
 pinky_02_l {parent=pinky_01_l, loc=<2.8274 0.0893 -0.1043>, rot=<0.0037 0.0011 0.2854 -0.9584>}
 pinky_03_l {parent=pinky_02_l, loc=<2.2046 0.0307 -0.0078>, rot=<-0.0002 0.0098 0.1445 -0.9895>}
 ring_metacarpal_l {parent=hand_l, loc=<2.2074 -0.3625 1.3638>, rot=<-0.0284 0.0722 -0.0458 -0.9959>}
 ring_01_l {parent=ring_metacarpal_l, loc=<7.0295 -0.6197 -0.1446>, rot=<-0.0489 -0.0252 0.1892 -0.9804>}
 ring_02_l {parent=ring_01_l, loc=<3.6174 0.4647 -0.1027>, rot=<0 0 0.2806 -0.9598>}
 ring_03_l {parent=ring_02_l, loc=<2.4152 0.114 0.0199>, rot=<0 0 0.1189 -0.9929>}
 thumb_01_l {parent=hand_l, loc=<1.3118 -0.3191 -2.459>, rot=<0.6546 -0.4042 0.0664 -0.6354>}
 thumb_02_l {parent=thumb_01_l, loc=<5.9336 -0.0448 0.0401>, rot=<-0.0037 0.0213 0.298 -0.9543>}
 thumb_03_l {parent=thumb_02_l, loc=<2.2223 0.3682 -0.0907>, rot=<0.0118 -0.013 0.0066 -0.9998>}
 lowerarm_twist_01_l {parent=lowerarm_l, loc=<18.316 0 0>, rot=<0 0 0 1>}
 lowerarm_twist_02_l {parent=lowerarm_l, loc=<10.6707 0 0>, rot=<0 0 0 1>}
 dyn_volume_base_lowerarm_l {parent=lowerarm_l, loc=<0 -4 0>, rot=<0 0 0 1>}
 dyn_volume_lowerarm_l {parent=dyn_volume_base_lowerarm_l, loc=<0 0 0>, rot=<0 0 0 1>}
 upperarm_twist_01_l {parent=upperarm_l, loc=<7.2135 -0.1134 0>, rot=<0 0 0 1>}
 upperarm_twist_02_l {parent=upperarm_l, loc=<19.7998 -0.1134 0>, rot=<0 0 0 1>}
 clavicle_r {parent=spine_05, loc=<5.4831 -0.1053 3.2946>, rot=<0.7289 0.0465 -0.6829 0.009>}
 upperarm_r {parent=clavicle_r, loc=<-9.8658 1.2136 0.4104>, rot=<0.0164 -0.3245 -0.0269 -0.9454>}
 lowerarm_r {parent=upperarm_r, loc=<-26.4651 0.8345 -0.259>, rot=<0.0036 0.0161 -0.2298 -0.9731>}
 hand_r {parent=lowerarm_r, loc=<-22.7356 0.3203 0.6106>, rot=<-0.6784 -0.0223 0.0735 -0.7307>}
 index_metacarpal_r {parent=hand_r, loc=<-1.6534 -0.0181 2.9601>, rot=<0.0158 -0.0124 -0.0173 -0.9996>}
 index_01_r {parent=index_metacarpal_r, loc=<-8.3008 -0.0357 0.0524>, rot=<-0.0044 0.0485 0.0924 -0.9945>}
 index_02_r {parent=index_01_r, loc=<-3.3589 0.3003 -0.0453>, rot=<-0.002 0.0027 0.3225 -0.9466>}
 index_03_r {parent=index_02_r, loc=<-2.5578 0.123 0.0252>, rot=<-0.002 0.0067 0.2558 -0.9667>}
 middle_metacarpal_r {parent=hand_r, loc=<-2.1856 0.1726 0.5984>, rot=<0.006 0.0168 -0.0329 -0.9993>}
 middle_01_r {parent=middle_metacarpal_r, loc=<-7.3273 0.2447 -0.0129>, rot=<-0.0208 0.0042 0.0858 -0.9961>}
 middle_02_r {parent=middle_01_r, loc=<-4.0124 -0.14 -0.1175>, rot=<-0.0218 0.0258 0.3512 -0.9357>}
 middle_03_r {parent=middle_02_r, loc=<-2.6077 -0.0533 -0.1373>, rot=<0.0082 0.0188 0.1708 -0.9851>}
 pinky_metacarpal_r {parent=hand_r, loc=<-1.7617 0.9753 -3.3814>, rot=<-0.0397 0.0814 -0.069 -0.9935>}
 pinky_01_r {parent=pinky_metacarpal_r, loc=<-7.1707 0.4832 0.1094>, rot=<-0.0654 -0.0035 0.2242 -0.9723>}
 pinky_02_r {parent=pinky_01_r, loc=<-2.8274 -0.0892 0.1043>, rot=<0.0037 0.0011 0.2854 -0.9584>}
 pinky_03_r {parent=pinky_02_r, loc=<-2.2045 -0.031 0.0078>, rot=<-0.0002 0.0098 0.1445 -0.9895>}
 ring_metacarpal_r {parent=hand_r, loc=<-2.2073 0.3627 -1.3638>, rot=<-0.0284 0.0722 -0.0458 -0.9959>}
 ring_01_r {parent=ring_metacarpal_r, loc=<-7.0297 0.6197 0.1446>, rot=<-0.0489 -0.0252 0.1892 -0.9804>}
 ring_02_r {parent=ring_01_r, loc=<-3.6171 -0.4648 0.1027>, rot=<0 0 0.2806 -0.9598>}
 ring_03_r {parent=ring_02_r, loc=<-2.4153 -0.1141 -0.0199>, rot=<0 0 0.1189 -0.9929>}
 thumb_01_r {parent=hand_r, loc=<-1.3119 0.3192 2.459>, rot=<0.6546 -0.4042 0.0664 -0.6354>}
 thumb_02_r {parent=thumb_01_r, loc=<-5.9335 0.0447 -0.0401>, rot=<-0.0037 0.0213 0.298 -0.9543>}
 thumb_03_r {parent=thumb_02_r, loc=<-2.2223 -0.3682 0.0907>, rot=<0.0118 -0.013 0.0066 -0.9998>}
 lowerarm_twist_01_r {parent=lowerarm_r, loc=<-18.3161 0 0>, rot=<0 0 0 1>}
 lowerarm_twist_02_r {parent=lowerarm_r, loc=<-10.6707 0 0>, rot=<0 0 0 1>}
 dyn_volume_base_lowerarm_r {parent=lowerarm_r, loc=<0 4 0>, rot=<0 0 0 1>}
 dyn_volume_lowerarm_r {parent=dyn_volume_base_lowerarm_r, loc=<0 0 0>, rot=<0 0 0 1>}
 upperarm_twist_01_r {parent=upperarm_r, loc=<-7.2135 0.1134 0>, rot=<0 0 0 1>}
 upperarm_twist_02_r {parent=upperarm_r, loc=<-19.7998 0.1134 0>, rot=<0 0 0 1>}
 neck_01 {parent=spine_05, loc=<11.3138 0.1529 0>, rot=<0 0 -0.2003 -0.9797>}
 neck_02 {parent=neck_01, loc=<4.5627 0 0>, rot=<0 0 0.0504 -0.9987>}
 head {parent=neck_02, loc=<5.3914 0 0>, rot=<0 0 0.115 -0.9934>}
 neckbrace_md_base {parent=neck_01, loc=<2.5379 4.7034 0>, rot=<-0.0116 -0.707 0.0116 -0.707>}
 neckbrace_md {parent=neckbrace_md_base, loc=<0 0 0>, rot=<0 0 0 1>}
 neckbrace_up {parent=neckbrace_md, loc=<0 0 1.6572>, rot=<0 0 0 1>}
 neckbrace_dn {parent=neckbrace_md, loc=<0 0 -1.4833>, rot=<0 0 0 1>}
 neckbrace_diamond {parent=neckbrace_md, loc=<0 0 0>, rot=<0 0 0 1>}
 dyn_collar_fr_base_l {parent=spine_05, loc=<6.9985 5.1417 -5.3332>, rot=<0.1847 -0.8161 0.0516 -0.5451>}
 dyn_collar_fr_l {parent=dyn_collar_fr_base_l, loc=<0 0 0>, rot=<0 0 0 1>}
 dyn_collar_bk_base {parent=spine_05, loc=<12.1735 -4.1283 0>, rot=<-0.0509 -0.7053 0.7053 0.0509>}
 dyn_collar_bk {parent=dyn_collar_bk_base, loc=<0 0 0>, rot=<0 0 0 1>}
 dyn_collar_side_base_l {parent=spine_05, loc=<11.3244 0.0279 -4.7279>, rot=<0.2378 -0.7876 0.0792 -0.5629>}
 dyn_collar_side_l {parent=dyn_collar_side_base_l, loc=<0 0 0>, rot=<0 0 0 1>}
 dyn_collar_fr_base_r {parent=spine_05, loc=<6.9985 5.1417 5.3332>, rot=<0.8161 0.1847 -0.5451 -0.0516>}
 dyn_collar_fr_r {parent=dyn_collar_fr_base_r, loc=<0 0 0>, rot=<0 0 0 1>}
 dyn_collar_side_base_r {parent=spine_05, loc=<11.3242 0.0278 4.7279>, rot=<0.7876 0.2378 -0.5629 -0.0792>}
 dyn_collar_side_r {parent=dyn_collar_side_base_r, loc=<0 0 0>, rot=<0 0 0 1>}
 dyn_button_c_r {parent=spine_05, loc=<3.8748 5.1122 9.949>, rot=<0.4441 -0.8549 -0.1084 -0.2452>}
 dyn_button_b_r {parent=spine_05, loc=<2.8506 7.5124 7.5364>, rot=<0.3679 -0.8727 -0.1416 -0.2879>}
 dyn_button_a_r {parent=spine_05, loc=<0.0536 8.0877 9.2004>, rot=<0.403 -0.8267 -0.1642 -0.3568>}
 dyn_button_a_l {parent=spine_05, loc=<0.9951 7.6471 -9.0482>, rot=<0.4136 -0.8262 0.2304 0.3055>}
 dyn_jacket_bk_a_base {parent=spine_04, loc=<2.6127 -7.7253 -0.0001>, rot=<0.0416 -0.9991 0 0>}
 dyn_jacket_bk_a {parent=dyn_jacket_bk_a_base, loc=<0 0 0>, rot=<0 0 0 1>}
 dyn_jacket_bk_b {parent=dyn_jacket_bk_a, loc=<6.5694 1.0377 0>, rot=<0 0 0 1>}
 dyn_jacket_side_a_base_l {parent=spine_04, loc=<0.1968 -4.7872 -9.0605>, rot=<0.0367 -0.8822 -0.469 0.0195>}
 dyn_jacket_side_a_l {parent=dyn_jacket_side_a_base_l, loc=<0 0 0>, rot=<0 0 0 1>}
 dyn_jacket_side_b_l {parent=dyn_jacket_side_a_l, loc=<6.3913 1.1889 0>, rot=<0 0 0 1>}
 jacket BL {controlType=EULER_TRANSFORM, parent=dyn_jacket_side_b_l, type=Control, loc=<4.1215 0 0>, rot=<0 0 0 1>, scale=<0.2 0.2 0.2>}
 dyn_jacket_fr_a_base_l {parent=spine_04, loc=<-0.0691 5.9205 -10.5644>, rot=<-0.0232 0.5576 0.8291 -0.0345>}
 dyn_jacket_fr_a_l {parent=dyn_jacket_fr_a_base_l, loc=<0 0 0>, rot=<0 0 0 1>}
 dyn_jacket_fr_b_l {parent=dyn_jacket_fr_a_l, loc=<6.2214 0 0>, rot=<0 0 0 1>}
 dyn_jacket_zipper_base {parent=dyn_jacket_fr_b_l, loc=<3.5965 -0.3788 4.8462>, rot=<0.9533 -0.0024 -0.1417 0.2667>}
 dyn_jacket_zipper {parent=dyn_jacket_zipper_base, loc=<0 0 0>, rot=<0 0 0 1>}
 "Jacket FL " {controlType=EULER_TRANSFORM, parent=dyn_jacket_fr_b_l, type=Control, loc=<6.4887 -0.6294 0.329>, rot=<0 0 0 1>, scale=<0.2 0.2 0.2>}
 dyn_chest_zipper_base {parent=spine_04, loc=<4.7221 11.1517 0.0201>, rot=<-0.0937 -0.9956 0 0>}
 dyn_chest_zipper {parent=dyn_chest_zipper_base, loc=<0 0 0>, rot=<0 0 0 1>}
 dyn_jacket_side_a_base_r {parent=spine_04, loc=<0.1971 -4.7872 9.0605>, rot=<0.8822 0.0367 0.0195 0.469>}
 dyn_jacket_side_a_r {parent=dyn_jacket_side_a_base_r, loc=<0 0 0>, rot=<0 0 0 1>}
 dyn_jacket_side_b_r {parent=dyn_jacket_side_a_r, loc=<-6.392 -1.1889 0>, rot=<0 0 0 1>}
 dyn_jacket_fr_a_base_r {parent=spine_04, loc=<-0.0688 5.9205 10.5644>, rot=<-0.5576 -0.0232 -0.0345 -0.8291>}
 dyn_jacket_fr_a_r {parent=dyn_jacket_fr_a_base_r, loc=<0 0 0>, rot=<0 0 0 1>}
 dyn_jacket_fr_b_r {parent=dyn_jacket_fr_a_r, loc=<-6.222 0 0>, rot=<0 0 0 1>}
 jacket FR {controlType=EULER_TRANSFORM, parent=dyn_jacket_fr_b_r, type=Control, loc=<0 0 0>, rot=<0 0 0 1>}
 Side body R {controlType=EULER_TRANSFORM, parent=spine_03, type=Control, loc=<3.8764 1.661 7.303>, rot=<0 0 0 1>, scale=<0.2 0.2 0.2>}
 back body {
  controlType=EULER_TRANSFORM
  parent=spine_03
  type=Control
  loc=<1.4982 -3.7156 -0.2654>
  rot=<0.1127 0.9931 0.0088 -0.0297>
  scale=<1.002 1 1>
 }
 Side body L {controlType=EULER_TRANSFORM, parent=spine_03, type=Control, loc=<2.2443 1.9203 -11.3312>, rot=<0 0 0 1>, scale=<0.2 0.2 0.2>}
 Side body L_2 {controlType=EULER_TRANSFORM, parent=spine_03, type=Control, loc=<0.745 -0.9197 -6.7001>, rot=<0.1124 0.8641 0.4902 -0.019>}
 spine_01_ctrl {controlType=EULER_TRANSFORM, parent=spine_01, type=Control, loc=<0 0 0>, rot=<0 0 0 1>}
 body FL {controlType=EULER_TRANSFORM, parent=spine_01, type=Control, loc=<2.745 0.616 -8.6377>, rot=<0 0 0 1>, scale=<0.2 0.2 0.2>}
 body FR {controlType=EULER_TRANSFORM, parent=spine_01, type=Control, loc=<3.0985 2.5871 9.1193>, rot=<0 0 0 1>, scale=<0.2 0.2 0.2>}
 thigh_l {parent=pelvis, loc=<-1.0855 -0.2109 -8.7053>, rot=<0.0009 0.0355 0.0155 -0.9993>}
 calf_l {parent=thigh_l, loc=<-42.5284 -0.0019 -0.7297>, rot=<0.0006 -0.0161 -0.0391 -0.9991>}
 calf_twist_01_l {parent=calf_l, loc=<-35.8483 -1.9197 -0.5794>, rot=<0.0237 0.0095 -0.0035 -0.9997>}
 calf_twist_02_l {parent=calf_l, loc=<-19.1546 -1.7943 -0.2638>, rot=<0.0237 0.0095 -0.0035 -0.9997>}
 foot_l {parent=calf_l, loc=<-46.7402 -2.5014 -0.592>, rot=<0.0427 -0.0238 0.0353 -0.9982>}
 ball_l {parent=foot_l, loc=<-3.5658 10.8066 -0.1544>, rot=<0.0001 -0.0001 -0.7063 -0.7079>}
 dyn_volume_base_calf_l {parent=calf_l, loc=<0 4 0>, rot=<0 0 0 1>}
 dyn_volume_calf_l {parent=dyn_volume_base_calf_l, loc=<0 0 0>, rot=<0 0 0 1>}
 dyn_volume_base_calf_bk_l {parent=calf_l, loc=<0 -6.4426 0>, rot=<0 0 0.0349 -0.9994>}
 dyn_volume_calf_bk_l {parent=dyn_volume_base_calf_bk_l, loc=<0 0 0>, rot=<0 0 0 1>}
 thigh_twist_01_l {parent=thigh_l, loc=<-27.952 0 0>, rot=<0.0013 0 0 -1>}
 dyn_volume_base_thigh_l {parent=thigh_l, loc=<-6 -9 0>, rot=<0 0 -0.4617 -0.887>}
 dyn_volume_thigh_l {parent=dyn_volume_base_thigh_l, loc=<0 0 0>, rot=<0 0 0 1>}
 thigh_r {parent=pelvis, loc=<-1.0853 -0.2109 8.7053>, rot=<0.0355 -0.0009 0.9993 0.0155>}
 calf_r {parent=thigh_r, loc=<42.5286 0.0019 0.7297>, rot=<0.0006 -0.0161 -0.0391 -0.9991>}
 calf_twist_01_r {parent=calf_r, loc=<35.8485 1.9197 0.5794>, rot=<0.0237 0.0095 -0.0035 -0.9997>}
 calf_twist_02_r {parent=calf_r, loc=<19.1546 1.7943 0.2638>, rot=<0.0237 0.0095 -0.0035 -0.9997>}
 foot_r {parent=calf_r, loc=<46.7403 2.5014 0.592>, rot=<0.0427 -0.0238 0.0353 -0.9982>}
 ball_r {parent=foot_r, loc=<3.5658 -10.8066 0.1544>, rot=<0.0001 -0.0001 -0.7063 -0.7079>}
 dyn_volume_base_calf_r {parent=calf_r, loc=<0 -4 0>, rot=<0 0 0 1>}
 dyn_volume_calf_r {parent=dyn_volume_base_calf_r, loc=<0 0 0>, rot=<0 0 0 1>}
 dyn_volume_calf_bk_base_r {parent=calf_r, loc=<0 6.4426 -0.0001>, rot=<0 0 0.0349 -0.9994>}
 dyn_volume_calf_bk_r {parent=dyn_volume_calf_bk_base_r, loc=<0 0 0>, rot=<0 0 0 1>}
 thigh_twist_01_r {parent=thigh_r, loc=<27.952 0 0>, rot=<0.0013 0 0 -1>}
 dyn_volume_base_thigh_r {parent=thigh_r, loc=<6 9 0>, rot=<0 0 -0.4617 -0.887>}
 dyn_volume_thigh_r {parent=dyn_volume_base_thigh_r, loc=<0 0 0>, rot=<0 0 0 1>}
 dyn_gun_base {parent=thigh_r, loc=<7.0643 0.9552 8.459>, rot=<0.9994 -0.0024 -0.0355 -0.0003>}
 dyn_gun {parent=dyn_gun_base, loc=<0 0 0>, rot=<0 0 0 1>}
 dyn_gun_scale {parent=dyn_gun, loc=<-0.6198 0.4077 -3.3653>, rot=<0.1533 0.0091 0.2175 -0.9639>}
]
variables [NoAnimDynamics {type=bool, default=False}]
