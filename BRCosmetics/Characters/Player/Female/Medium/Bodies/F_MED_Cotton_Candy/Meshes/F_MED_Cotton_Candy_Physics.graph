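# Physics asset for the F_MED_Cotton_Candy female character body, set up for a dynamic collar with back, front and side panels.
# Each collar panel is a box body constrained to a small capsule on its base bone with tight swing and twist limits.
# Capsules on the head, neck and upper arms supply surrounding body shapes, with head-to-neck rotation left free.
# The front-right panel alone explicitly locks twist, unlike its left counterpart.
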PhysicsAsset /BRCosmetics/Characters/Player/Female/Medium/Bodies/F_MED_Cotton_Candy/Meshes/F_MED_Cotton_Candy_Physics
bodies {
 dyn_collar_bk [{shapeType=Box, extentX=6.5, extentY=11, extentZ=2.5, center=<-2.1574 0 0>}]
 dyn_collar_bk_base [{shapeType=Capsule, length=1, radius=1}]
 dyn_collar_fr_base_l [{shapeType=Capsule, length=1, radius=1}]
 dyn_collar_fr_base_r [{shapeType=Capsule, length=1, radius=1}]
 dyn_collar_fr_l [{shapeType=Box, extentX=2.5, extentY=13.5, extentZ=9.5, center=<-0.0001 0.0001 3.8781>}]
 dyn_collar_fr_r [{shapeType=Box, extentX=2.5, extentY=13.5, extentZ=9.5, center=<0.0001 -0.0001 -3.8781>, rotation=<rot 0 0 -180>}]
 dyn_collar_side_base_l [{shapeType=Capsule, length=1, radius=1}]
 dyn_collar_side_base_r [{shapeType=Capsule, length=1, radius=1}]
 dyn_collar_side_l [{shapeType=Box, extentX=2.5, extentY=11, extentZ=7, center=<0 0 1.7364>}]
 dyn_collar_side_r [{shapeType=Box, extentX=2.5, extentY=11, extentZ=7, center=<0 0 -1.7364>, rotation=<rot 0 0 180>}]
 head [{shapeType=Capsule, length=4.75, radius=8, center=<2.7121 1.3494 0>, rotation=<rot 90 0 0.5441>}]
 neck_01 [{shapeType=Capsule, length=2, radius=4.25, center=<3.6526 0.1952 0>, rotation=<rot 90 0 0>}]
 upperarm_l [{shapeType=Capsule, length=16.5, radius=4.25, center=<7.1401 -0.0001 -0.3797>, rotation=<rot 86.7569 0 0>}]
 upperarm_r [{shapeType=Capsule, length=16.5, radius=4.25, center=<-7.1401 0.0001 0.3797>, rotation=<rot 86.7569 0 -180>}]
}
constraints [
 {
  bone1Name=head
  bone2Name=neck_01
  swing1LimitDegrees=45
  swing1Motion=Free
  swing2LimitDegrees=45
  swing2Motion=Free
  twistLimitDegrees=45
  twistMotion=Free
 }
 {bone1Name=dyn_collar_fr_l, bone2Name=dyn_collar_fr_base_l, swing1LimitDegrees=10, swing2LimitDegrees=10, twistLimitDegrees=5}
 {bone1Name=dyn_collar_bk, bone2Name=dyn_collar_bk_base, swing1LimitDegrees=10, swing2LimitDegrees=10, twistLimitDegrees=5}
 {bone1Name=dyn_collar_side_l, bone2Name=dyn_collar_side_base_l, swing1LimitDegrees=10, swing2LimitDegrees=10, twistLimitDegrees=5}
 {
  bone1Name=dyn_collar_fr_r
  bone2Name=dyn_collar_fr_base_r
  swing1LimitDegrees=10
  swing2LimitDegrees=10
  twistLimitDegrees=5
  twistMotion=Locked
 }
 {bone1Name=dyn_collar_side_r, bone2Name=dyn_collar_side_base_r, swing1LimitDegrees=10, swing2LimitDegrees=10, twistLimitDegrees=5}
]
