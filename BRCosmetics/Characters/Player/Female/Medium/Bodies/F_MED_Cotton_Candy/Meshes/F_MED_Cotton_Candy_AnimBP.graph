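# Cosmetic AnimBP for the Cotton Candy female body, copying the attached parent's pose and adding secondary motion.
# Event graph measures elbow, knee, thigh, upper-arm and head yaw against ref pose, driving corrective volume, neckbrace and collar bones alongside weighted constraints.
# Collar also shifts while skydive-floating; AnimDynamics, RigidBody and Control Rig follow, the rig receiving NoAnimDynamics when either physics flag is off.
# Some front-calf, neckbrace and upper-jacket nodes are unconnected.
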AnimBlueprint /BRCosmetics/Characters/Player/Female/Medium/Bodies/F_MED_Cotton_Candy/Meshes/F_MED_Cotton_Candy_AnimBP
target_skeleton=&/F_MED_Cotton_Candy_Skeleton
graphs [
 AnimGraph {nodes [
   Anim Dynamics - Bone: dyn_jacket_side_a_r(Alpha=0, ExternalForce="0,0,0") {
    angularDampingOverride=1
    angularSpringConstant=1500
    bAngularSpring=true
    bGravityOverrideInSimSpace=true
    bOverrideAngularDamping=true
    gravityScale=0.35
    simulationSpace=RootRelative
    alphaBoolBlend {blendOption=Linear}
    boundBone {boneName=dyn_jacket_side_a_r}
    physicsBodyDefinitions [{
      sphereCollisionRadius=0
      boundBone {boneName=dyn_jacket_side_a_r}
      boxExtents=<10 20 20>
      constraintSetup {angularLimitsMax=<0 0 0>, angularLimitsMin=<0 0 -5>, angularTarget=<0 0 0>}
      localJointOffset=<-10 0 0>
     }]
   }
   Anim Dynamics - Bone: dyn_jacket_bk_a(Alpha=0, ExternalForce="0,0,0") {
    angularDampingOverride=1
    angularSpringConstant=1500
    bAngularSpring=true
    bGravityOverrideInSimSpace=true
    bOverrideAngularDamping=true
    gravityScale=0.35
    simulationSpace=RootRelative
    alphaBoolBlend {blendOption=Linear}
    boundBone {boneName=dyn_jacket_bk_a}
    physicsBodyDefinitions [{
      sphereCollisionRadius=0
      boundBone {boneName=dyn_jacket_bk_a}
      boxExtents=<10 20 20>
      constraintSetup {angularLimitsMax=<0 0 0>, angularLimitsMin=<0 0 -5>, angularTarget=<0 0 0>}
      localJointOffset=<10 0 0>
     }]
   }
   Anim Dynamics - Bone: dyn_jacket_fr_a_l(Alpha=0, ExternalForce="0,0,0") {
    angularDampingOverride=1
    angularSpringConstant=1500
    bAngularSpring=true
    bGravityOverrideInSimSpace=true
    bOverrideAngularDamping=true
    gravityScale=0.35
    simulationSpace=RootRelative
    alphaBoolBlend {blendOption=Linear}
    boundBone {boneName=dyn_jacket_fr_a_l}
    physicsBodyDefinitions [{
      sphereCollisionRadius=0
      boundBone {boneName=dyn_jacket_fr_a_l}
      boxExtents=<10 20 20>
      constraintSetup {angularLimitsMax=<0 0 0>, angularLimitsMin=<0 0 -5>, angularTarget=<0 0 0>}
      localJointOffset=<10 0 0>
     }]
   }
   Anim Dynamics - Bone: dyn_jacket_fr_a_r(Alpha=0, ExternalForce="0,0,0") {
    angularDampingOverride=1
    angularSpringConstant=1500
    bAngularSpring=true
    bGravityOverrideInSimSpace=true
    bOverrideAngularDamping=true
    gravityScale=0.35
    simulationSpace=RootRelative
    alphaBoolBlend {blendOption=Linear}
    boundBone {boneName=dyn_jacket_fr_a_r}
    physicsBodyDefinitions [{
      sphereCollisionRadius=0
      boundBone {boneName=dyn_jacket_fr_a_r}
      boxExtents=<10 20 20>
      constraintSetup {angularLimitsMax=<0 0 0>, angularLimitsMin=<0 0 -5>, angularTarget=<0 0 0>}
      localJointOffset=<-10 0 0>
     }]
   }
   Anim Dynamics - Bone: dyn_jacket_side_a_l(Alpha=0, ExternalForce="0,0,0") {
    angularDampingOverride=1
    angularSpringConstant=1500
    bAngularSpring=true
    bGravityOverrideInSimSpace=true
    bOverrideAngularDamping=true
    gravityScale=0.35
    simulationSpace=RootRelative
    alphaBoolBlend {blendOption=Linear}
    boundBone {boneName=dyn_jacket_side_a_l}
    physicsBodyDefinitions [{
      sphereCollisionRadius=0
      boundBone {boneName=dyn_jacket_side_a_l}
      boxExtents=<10 20 20>
      constraintSetup {angularLimitsMax=<0 0 0>, angularLimitsMin=<0 0 -5>, angularTarget=<0 0 0>}
      localJointOffset=<10 0 0>
     }]
   }
   n1 = "Constraint - dyn_volume_base_calf_l (PP)"!(ComponentPose, Alpha=1, ConstraintWeights_0=0.5, ConstraintWeights_1:float=0.5) -> (Pose) {
    alphaBoolBlend {blendOption=Linear}
    boneToModify {boneName=dyn_volume_base_calf_l}
    constraintSetup [{targetBone {boneName=thigh_l}}, {targetBone {boneName=calf_l}}]
    constraintWeights [1,1]
   }
   n2 = "Constraint - dyn_volume_base_calf_r (PP)"!(ComponentPose=@n1, Alpha=1, ConstraintWeights_0=0.5, ConstraintWeights_1:float=0.5) -> (Pose) {
    alphaBoolBlend {blendOption=Linear}
    boneToModify {boneName=dyn_volume_base_calf_r}
    constraintSetup [{targetBone {boneName=thigh_r}}, {targetBone {boneName=calf_r}}]
    constraintWeights [1,1]
   }
   n3 = "Transform (Modify) Bone - Bone: dyn_volume_calf_l"(ComponentPose=@n2, Alpha=Get knee_rot_l() -> (knee_rot_l:double), Translation="1,3,0", Rotation="0,0,0", Scale="1.5,1,1.5") {
    rotationMode=BMM_Additive
    rotationSpace=BCS_BoneSpace
    scaleMode=BMM_Additive
    scaleSpace=BCS_BoneSpace
    translationMode=BMM_Additive
    translationSpace=BCS_BoneSpace
    alphaBoolBlend {blendOption=Linear}
    alphaScaleBiasClamp {bMapRange=true, inRange {max=80}}
    boneToModify {boneName=dyn_volume_calf_l}
    scale {x=1.5, z=1.5}
    translation=<1 3>
   }
   "Transform (Modify) Bone - Bone: dyn_volume_calf_r"(ComponentPose=@n3, Alpha=Get knee_rot_r() -> (knee_rot_r:double), Translation="-1,-3,0", Rotation="0,0,0", Scale="1.5,1,1.5") {
    rotationMode=BMM_Additive
    rotationSpace=BCS_BoneSpace
    scaleMode=BMM_Additive
    scaleSpace=BCS_BoneSpace
    translationMode=BMM_Additive
    translationSpace=BCS_BoneSpace
    alphaBoolBlend {blendOption=Linear}
    alphaScaleBiasClamp {bMapRange=true, inRange {max=80}}
    boneToModify {boneName=dyn_volume_calf_r}
    scale {x=1.5, z=1.5}
    translation=<-1 -3>
   }
   n4 = "Transform (Modify) Bone - Bone: neckbrace_up_base"(Alpha=1, Translation="0,0,-2.5", Rotation="0,0,0", Scale="1,1,1") {
    rotationMode=BMM_Additive
    rotationSpace=BCS_BoneSpace
    scaleMode=BMM_Additive
    scaleSpace=BCS_BoneSpace
    translationMode=BMM_Additive
    translationSpace=BCS_BoneSpace
    alphaBoolBlend {blendOption=Linear}
    alphaScaleBiasClamp {bMapRange=true, inRange {max=0, min=-30}, outRange {max=0, min=1}}
    boneToModify {boneName=neckbrace_up_base}
    translation {z=-2.5}
   }
   n5 = "Transform (Modify) Bone - Bone: neckbrace_dn"(ComponentPose=@n4, Alpha=1, Translation="0,0.5,0.5", Rotation="0,0,0", Scale="1,1,1") {
    rotationMode=BMM_Additive
    rotationSpace=BCS_BoneSpace
    scaleMode=BMM_Additive
    scaleSpace=BCS_BoneSpace
    translationMode=BMM_Additive
    translationSpace=BCS_BoneSpace
    alphaBoolBlend {blendOption=Linear}
    alphaScaleBiasClamp {bMapRange=true, inRange {max=0, min=-30}, outRange {max=0, min=1}}
    boneToModify {boneName=neckbrace_dn}
    translation {y=0.5, z=0.5}
   }
   "Transform (Modify) Bone - Bone: neckbrace_dn"(ComponentPose=@n5, Alpha=1, Translation="0,1.5,1", Rotation="0,0,0", Scale="1,1,1") {
    rotationMode=BMM_Additive
    rotationSpace=BCS_BoneSpace
    scaleMode=BMM_Additive
    scaleSpace=BCS_BoneSpace
    translationMode=BMM_Additive
    translationSpace=BCS_BoneSpace
    alphaBoolBlend {blendOption=Linear}
    alphaScaleBiasClamp {bMapRange=true, inRange {max=20}}
    boneToModify {boneName=neckbrace_dn}
    translation {y=1.5, z=1}
   }
   n6 = "Constraint - dyn_volume_base_lowerarm_l (PP)"!(
    ComponentPose=Local To Component(LocalPose=Copy Pose From Mesh() {bUseAttachedParent=true})
    Alpha=1
    ConstraintWeights_0=0.65
    ConstraintWeights_1:float=0.35
   ) -> (Pose) {
    alphaBoolBlend {blendOption=Linear}
    boneToModify {boneName=dyn_volume_base_lowerarm_l}
    constraintSetup [{targetBone {boneName=upperarm_l}}, {targetBone {boneName=lowerarm_l}}]
    constraintWeights [1,1]
   }
   n7 = "Constraint - dyn_volume_base_lowerarm_r (PP)"!(ComponentPose=@n6, Alpha=1, ConstraintWeights_0=0.65, ConstraintWeights_1:float=0.35) -> (Pose) {
    alphaBoolBlend {blendOption=Linear}
    boneToModify {boneName=dyn_volume_base_lowerarm_r}
    constraintSetup [{targetBone {boneName=upperarm_r}}, {targetBone {boneName=lowerarm_r}}]
    constraintWeights [1,1]
   }
   n8 = "Transform (Modify) Bone - Bone: dyn_volume_lowerarm_l"(ComponentPose=@n7, Alpha=Get elbow_rot_l() -> (elbow_rot_l:double), Translation="0,-2,0", Rotation="0,0,0", Scale="0.3,1,1") {
    rotationMode=BMM_Additive
    rotationSpace=BCS_BoneSpace
    scaleMode=BMM_Additive
    scaleSpace=BCS_BoneSpace
    translationMode=BMM_Additive
    translationSpace=BCS_BoneSpace
    alphaBoolBlend {blendOption=Linear}
    alphaScaleBiasClamp {bMapRange=true, inRange {max=80}}
    boneToModify {boneName=dyn_volume_lowerarm_l}
    scale {x=0.3}
    translation {y=-2}
   }
   n9 = "Transform (Modify) Bone - Bone: dyn_volume_lowerarm_r"(ComponentPose=@n8, Alpha=Get elbow_rot_r() -> (elbow_rot_r:double), Translation="0,2,0", Rotation="0,0,0", Scale="0.3,1,1") {
    rotationMode=BMM_Additive
    rotationSpace=BCS_BoneSpace
    scaleMode=BMM_Additive
    scaleSpace=BCS_BoneSpace
    translationMode=BMM_Additive
    translationSpace=BCS_BoneSpace
    alphaBoolBlend {blendOption=Linear}
    alphaScaleBiasClamp {bMapRange=true, inRange {max=80}}
    boneToModify {boneName=dyn_volume_lowerarm_r}
    scale {x=0.3}
    translation {y=2}
   }
   n10 = "Constraint - dyn_volume_base_thigh_l (PP)"!(ComponentPose=@n9, Alpha=1, ConstraintWeights_0=0.6, ConstraintWeights_1:float=0.4) -> (Pose) {
    alphaBoolBlend {blendOption=Linear}
    boneToModify {boneName=dyn_volume_base_thigh_l}
    constraintSetup [{targetBone {boneName=pelvis}}, {targetBone {boneName=thigh_l}}]
    constraintWeights [1,1]
   }
   n11 = "Constraint - dyn_volume_base_thigh_r (PP)"!(ComponentPose=@n10, Alpha=1, ConstraintWeights_0=0.6, ConstraintWeights_1:float=0.4) -> (Pose) {
    alphaBoolBlend {blendOption=Linear}
    boneToModify {boneName=dyn_volume_base_thigh_r}
    constraintSetup [{targetBone {boneName=pelvis}}, {targetBone {boneName=thigh_r}}]
    constraintWeights [1,1]
   }
   n12 = Get thigh_rot_l()
   n13 = "Transform (Modify) Bone - Bone: dyn_volume_thigh_l"(ComponentPose=@n11, Alpha=@n12, Translation="-4,-2,0", Rotation="0,0,0", Scale="1,1,1") {
    rotationMode=BMM_Additive
    rotationSpace=BCS_BoneSpace
    scaleMode=BMM_Additive
    scaleSpace=BCS_BoneSpace
    translationMode=BMM_Additive
    translationSpace=BCS_BoneSpace
    alphaBoolBlend {blendOption=Linear}
    alphaScaleBiasClamp {bMapRange=true, inRange {max=0, min=-100}, outRange {max=0, min=1}}
    boneToModify {boneName=dyn_volume_thigh_l}
    translation=<-4 -2>
   }
   n14 = "Transform (Modify) Bone - Bone: dyn_volume_thigh_l"(ComponentPose=@n13, Alpha=@n12, Translation="0,-4,0", Rotation="0,0,0", Scale="1,1,1") {
    rotationMode=BMM_Additive
    rotationSpace=BCS_BoneSpace
    scaleMode=BMM_Additive
    scaleSpace=BCS_BoneSpace
    translationMode=BMM_Additive
    translationSpace=BCS_BoneSpace
    alphaBoolBlend {blendOption=Linear}
    alphaScaleBiasClamp {bMapRange=true, inRange {max=50}}
    boneToModify {boneName=dyn_volume_thigh_l}
    translation {y=-4}
   }
   n15 = Get thigh_rot_r()
   n16 = "Transform (Modify) Bone - Bone: dyn_volume_thigh_r"(ComponentPose=@n14, Alpha=@n15, Translation="4,2,0", Rotation="0,0,0", Scale="1,1,1") {
    rotationMode=BMM_Additive
    rotationSpace=BCS_BoneSpace
    scaleMode=BMM_Additive
    scaleSpace=BCS_BoneSpace
    translationMode=BMM_Additive
    translationSpace=BCS_BoneSpace
    alphaBoolBlend {blendOption=Linear}
    alphaScaleBiasClamp {bMapRange=true, inRange {max=0, min=-100}, outRange {max=0, min=1}}
    boneToModify {boneName=dyn_volume_thigh_r}
    translation=<4 2>
   }
   n17 = "Transform (Modify) Bone - Bone: dyn_volume_thigh_r"(ComponentPose=@n16, Alpha=@n15, Translation="0,4,0", Rotation="0,0,0", Scale="1,1,1") {
    rotationMode=BMM_Additive
    rotationSpace=BCS_BoneSpace
    scaleMode=BMM_Additive
    scaleSpace=BCS_BoneSpace
    translationMode=BMM_Additive
    translationSpace=BCS_BoneSpace
    alphaBoolBlend {blendOption=Linear}
    alphaScaleBiasClamp {bMapRange=true, inRange {max=50}}
    boneToModify {boneName=dyn_volume_thigh_r}
    translation {y=4}
   }
   n18 = "Constraint - dyn_volume_base_calf_bk_l (PP)"!(ComponentPose=@n17, Alpha=1, ConstraintWeights_0=0.5, ConstraintWeights_1:float=0.5) -> (Pose) {
    alphaBoolBlend {blendOption=Linear}
    boneToModify {boneName=dyn_volume_base_calf_bk_l}
    constraintSetup [{targetBone {boneName=thigh_l}}, {targetBone {boneName=calf_l}}]
    constraintWeights [1,1]
   }
   n19 = "Constraint - dyn_volume_calf_bk_base_r (PP)"!(ComponentPose=@n18, Alpha=1, ConstraintWeights_0=0.5, ConstraintWeights_1:float=0.5) -> (Pose) {
    alphaBoolBlend {blendOption=Linear}
    boneToModify {boneName=dyn_volume_calf_bk_base_r}
    constraintSetup [{targetBone {boneName=thigh_r}}, {targetBone {boneName=calf_r}}]
    constraintWeights [1,1]
   }
   n20 = "Transform (Modify) Bone - Bone: dyn_volume_calf_bk_l"(ComponentPose=@n19, Alpha=Get knee_rot_l(), Translation="1,-8,0", Rotation="0,0,0", Scale="1,1,1") {
    rotationMode=BMM_Additive
    rotationSpace=BCS_BoneSpace
    scaleMode=BMM_Additive
    scaleSpace=BCS_BoneSpace
    translationMode=BMM_Additive
    translationSpace=BCS_BoneSpace
    alphaBoolBlend {blendOption=Linear}
    alphaScaleBiasClamp {bMapRange=true, inRange {max=80}}
    boneToModify {boneName=dyn_volume_calf_bk_l}
    translation=<1 -8>
   }
   n21 = "Transform (Modify) Bone - Bone: dyn_volume_calf_bk_r"(ComponentPose=@n20, Alpha=Get knee_rot_r(), Translation="0,8,0", Rotation="0,0,0", Scale="1,1,1") {
    rotationMode=BMM_Additive
    rotationSpace=BCS_BoneSpace
    scaleMode=BMM_Additive
    scaleSpace=BCS_BoneSpace
    translationMode=BMM_Additive
    translationSpace=BCS_BoneSpace
    alphaBoolBlend {blendOption=Linear}
    alphaScaleBiasClamp {bMapRange=true, inRange {max=80}}
    boneToModify {boneName=dyn_volume_calf_bk_r}
    translation {y=8}
   }
   n22 = "Constraint - dyn_button_a_l (PPP)"!(ComponentPose=@n21, Alpha=1, ConstraintWeights_0=0.3717, ConstraintWeights_1:float=0.4574, ConstraintWeights_2:float=0.1709) -> (Pose) {
    alphaBoolBlend {blendOption=Linear}
    boneToModify {boneName=dyn_button_a_l}
    constraintSetup [{targetBone {boneName=spine_04}}, {targetBone {boneName=spine_05}}, {targetBone {boneName=clavicle_l}}]
    constraintWeights [1,1,1]
   }
   n23 = "Constraint - dyn_button_a_r (PPP)"!(ComponentPose=@n22, Alpha=1, ConstraintWeights_0=0.1424, ConstraintWeights_1:float=0.4405, ConstraintWeights_2:float=0.4171) -> (Pose) {
    alphaBoolBlend {blendOption=Linear}
    boneToModify {boneName=dyn_button_a_r}
    constraintSetup [{targetBone {boneName=clavicle_r}}, {targetBone {boneName=spine_04}}, {targetBone {boneName=spine_05}}]
    constraintWeights [1,1,1]
   }
   n24 = "Constraint - dyn_button_b_r (PPP)"!(ComponentPose=@n23, Alpha=1, ConstraintWeights_0=0.1243, ConstraintWeights_1:float=0.2261, ConstraintWeights_2:float=0.6496) -> (Pose) {
    alphaBoolBlend {blendOption=Linear}
    boneToModify {boneName=dyn_button_b_r}
    constraintSetup [{targetBone {boneName=clavicle_r}}, {targetBone {boneName=spine_04}}, {targetBone {boneName=spine_05}}]
    constraintWeights [1,1,1]
   }
   n25 = "Constraint - dyn_button_c_r (PPPP)"!(
    ComponentPose=@n24
    Alpha=1
    ConstraintWeights_0=0.3843
    ConstraintWeights_1:float=0.0295
    ConstraintWeights_2:float=0.2107
    ConstraintWeights_3:float=0.3754
   ) -> (Pose) {
    alphaBoolBlend {blendOption=Linear}
    boneToModify {boneName=dyn_button_c_r}
    constraintSetup [
     {targetBone {boneName=clavicle_r}}
     {targetBone {boneName=upperarm_twist_01_r}}
     {targetBone {boneName=spine_04}}
     {targetBone {boneName=spine_05}}
    ]
    constraintWeights [1,1,1,1]
   }
   n26 = "Constraint - dyn_jacket_fr_a_base_l (PPPPP)"!(
    ComponentPose=@n25
    Alpha=1
    ConstraintWeights_0=0.0331
    ConstraintWeights_1:float=0.0319
    ConstraintWeights_2:float=0.7325
    ConstraintWeights_3:float=0.1225
    ConstraintWeights_4:float=0.08
   ) -> (Pose) {
    alphaBoolBlend {blendOption=Linear}
    boneToModify {boneName=dyn_jacket_fr_a_base_l}
    constraintSetup [
     {targetBone {boneName=spine_02}}
     {targetBone {boneName=spine_03}}
     {targetBone {boneName=spine_04}}
     {targetBone {boneName=spine_05}}
     {targetBone {boneName=clavicle_l}}
    ]
    constraintWeights [1,1,1,1,1]
   }
   n27 = "Constraint - dyn_jacket_fr_a_base_r (PPPPP)"!(
    ComponentPose=@n26
    Alpha=1
    ConstraintWeights_0=0.08
    ConstraintWeights_1:float=0.0331
    ConstraintWeights_2:float=0.0319
    ConstraintWeights_3:float=0.7325
    ConstraintWeights_4:float=0.1225
   ) -> (Pose) {
    alphaBoolBlend {blendOption=Linear}
    boneToModify {boneName=dyn_jacket_fr_a_base_r}
    constraintSetup [
     {targetBone {boneName=clavicle_r}}
     {targetBone {boneName=spine_02}}
     {targetBone {boneName=spine_03}}
     {targetBone {boneName=spine_04}}
     {targetBone {boneName=spine_05}}
    ]
    constraintWeights [1,1,1,1,1]
   }
   n28 = "Constraint - dyn_jacket_side_a_base_l (PPPPPP)"!(
    ComponentPose=@n27
    Alpha=1
    ConstraintWeights_0=0.2327
    ConstraintWeights_1:float=0.0557
    ConstraintWeights_2:float=0.034
    ConstraintWeights_3:float=0.3957
    ConstraintWeights_4:float=0.2325
    ConstraintWeights_5:float=0.0495
   ) -> (Pose) {
    alphaBoolBlend {blendOption=Linear}
    boneToModify {boneName=dyn_jacket_side_a_base_l}
    constraintSetup [
     {targetBone {boneName=clavicle_l}}
     {targetBone {boneName=spine_02}}
     {targetBone {boneName=spine_03}}
     {targetBone {boneName=spine_04}}
     {targetBone {boneName=spine_05}}
     {targetBone {boneName=upperarm_twist_01_l}}
    ]
    constraintWeights [1,1,1,1,1,1]
   }
   n29 = "Constraint - dyn_jacket_side_a_base_r (PPPPPP)"!(
    ComponentPose=@n28
    Alpha=1
    ConstraintWeights_0=0.0495
    ConstraintWeights_1:float=0.0557
    ConstraintWeights_2:float=0.034
    ConstraintWeights_3:float=0.3957
    ConstraintWeights_4:float=0.2325
    ConstraintWeights_5:float=0.2327
   ) -> (Pose) {
    alphaBoolBlend {blendOption=Linear}
    boneToModify {boneName=dyn_jacket_side_a_base_r}
    constraintSetup [
     {targetBone {boneName=upperarm_twist_01_r}}
     {targetBone {boneName=spine_02}}
     {targetBone {boneName=spine_03}}
     {targetBone {boneName=spine_04}}
     {targetBone {boneName=spine_05}}
     {targetBone {boneName=clavicle_r}}
    ]
    constraintWeights [1,1,1,1,1,1]
   }
   n30 = "Constraint - dyn_jacket_bk_a_base (PPPPP)"!(
    ComponentPose=@n29
    Alpha=1
    ConstraintWeights_0=0.0177
    ConstraintWeights_1:float=0.0622
    ConstraintWeights_2:float=0.0662
    ConstraintWeights_3:float=0.408
    ConstraintWeights_4:float=0.4459
   ) -> (Pose) {
    alphaBoolBlend {blendOption=Linear}
    boneToModify {boneName=dyn_jacket_bk_a_base}
    constraintSetup [
     {targetBone {boneName=spine_01}}
     {targetBone {boneName=spine_02}}
     {targetBone {boneName=spine_03}}
     {targetBone {boneName=spine_04}}
     {targetBone {boneName=spine_05}}
    ]
    constraintWeights [1,1,1,1,1]
   }
   n31 = "Constraint - dyn_gun_base (PPPP)"!(
    ComponentPose=@n30
    Alpha=1
    ConstraintWeights_0=0.1206
    ConstraintWeights_1:float=0.1077
    ConstraintWeights_2:float=0.7627
    ConstraintWeights_3:float=0.0089
   ) -> (Pose) {
    alphaBoolBlend {blendOption=Linear}
    boneToModify {boneName=dyn_gun_base}
    constraintSetup [
     {targetBone {boneName=thigh_r}}
     {targetBone {boneName=pelvis}}
     {targetBone {boneName=thigh_twist_01_r}}
     {targetBone {boneName=dyn_volume_thigh_r}}
    ]
    constraintWeights [1,1,1,1]
   }
   n32 = "Constraint - dyn_chest_zipper_base (PP)"!(ComponentPose=@n31, Alpha=1, ConstraintWeights_0=0.62, ConstraintWeights_1:float=0.38) -> (Pose) {
    alphaBoolBlend {blendOption=Linear}
    boneToModify {boneName=dyn_chest_zipper_base}
    constraintSetup [{targetBone {boneName=spine_04}}, {targetBone {boneName=spine_05}}]
    constraintWeights [1,1]
   }
   n33 = "Constraint - dyn_jacket_zipper_base (PPPPP)"!(
    ComponentPose=@n32
    Alpha=1
    ConstraintWeights_0=0.0801
    ConstraintWeights_1:float=0.0227
    ConstraintWeights_2:float=0.1561
    ConstraintWeights_3:float=0.0283
    ConstraintWeights_4:float=0.7128
   ) -> (Pose) {
    alphaBoolBlend {blendOption=Linear}
    boneToModify {boneName=dyn_jacket_zipper_base}
    constraintSetup [
     {targetBone {boneName=dyn_jacket_fr_a_l}}
     {targetBone {boneName=spine_02}}
     {targetBone {boneName=spine_04}}
     {targetBone {boneName=spine_05}}
     {targetBone {boneName=dyn_jacket_fr_b_l}}
    ]
    constraintWeights [1,1,1,1,1]
   }
   n34 = "Constraint - neckbrace_md_base (PPPP)"!(
    ComponentPose=@n33
    Alpha=1
    ConstraintWeights_0=0.25
    ConstraintWeights_1:float=0.25
    ConstraintWeights_2:float=0.15
    ConstraintWeights_3:float=0.35
   ) -> (Pose) {
    alphaBoolBlend {blendOption=Linear}
    boneToModify {boneName=neckbrace_md_base}
    constraintSetup [{targetBone {boneName=neck_02}}, {targetBone {boneName=head}}, {targetBone {boneName=neck_01}}, {targetBone {boneName=spine_05}}]
    constraintWeights [1,1,1,1]
   }
   n35 = Get HeadUpDn() -> (HeadUpDn:double)
   n36 = "Transform (Modify) Bone - Bone: neckbrace_md"(ComponentPose=@n34, Alpha=@n35, Translation="0,0,-1", Rotation="0,0,0", Scale="1,1,1") {
    rotationMode=BMM_Additive
    rotationSpace=BCS_BoneSpace
    scaleMode=BMM_Additive
    scaleSpace=BCS_BoneSpace
    translationMode=BMM_Additive
    translationSpace=BCS_BoneSpace
    alphaBoolBlend {blendOption=Linear}
    alphaScaleBiasClamp {bMapRange=true, inRange {max=0, min=-30}, outRange {max=0, min=1}}
    boneToModify {boneName=neckbrace_md}
    translation {z=-1}
   }
   n37 = "Transform (Modify) Bone - Bone: neckbrace_up"(ComponentPose=@n36, Alpha=@n35, Translation="0,1,-2", Rotation="0,0,20", Scale="1.5,1,1") {
    rotationMode=BMM_Additive
    rotationSpace=BCS_BoneSpace
    scaleMode=BMM_Additive
    scaleSpace=BCS_BoneSpace
    translationMode=BMM_Additive
    translationSpace=BCS_BoneSpace
    alphaBoolBlend {blendOption=Linear}
    alphaScaleBiasClamp {bMapRange=true, inRange {max=0, min=-30}, outRange {max=0, min=1}}
    boneToModify {boneName=neckbrace_up}
    rotation {roll=20}
    scale {x=1.5}
    translation {y=1, z=-2}
   }
   n38 = "Transform (Modify) Bone - Bone: neckbrace_dn"(ComponentPose=@n37, Alpha=@n35, Translation="0,1,-0.5", Rotation="0,0,20", Scale="1.5,1,1") {
    rotationMode=BMM_Additive
    rotationSpace=BCS_BoneSpace
    scaleMode=BMM_Additive
    scaleSpace=BCS_BoneSpace
    translationMode=BMM_Additive
    translationSpace=BCS_BoneSpace
    alphaBoolBlend {blendOption=Linear}
    alphaScaleBiasClamp {bMapRange=true, inRange {max=0, min=-30}, outRange {max=0, min=1}}
    boneToModify {boneName=neckbrace_dn}
    rotation {roll=20}
    scale {x=1.5}
    translation {y=1, z=-0.5}
   }
   n39 = Get HeadUpDn()
   n40 = "Transform (Modify) Bone - Bone: neckbrace_md"(ComponentPose=@n38, Alpha=@n39, Translation="0,1.5,0.5", Rotation="0,0,-40", Scale="1,1,1") {
    rotationMode=BMM_Additive
    rotationSpace=BCS_BoneSpace
    scaleMode=BMM_Additive
    scaleSpace=BCS_BoneSpace
    translationMode=BMM_Additive
    translationSpace=BCS_BoneSpace
    alphaBoolBlend {blendOption=Linear}
    alphaScaleBiasClamp {bMapRange=true, inRange {max=20}}
    boneToModify {boneName=neckbrace_md}
    rotation {roll=-40}
    translation {y=1.5, z=0.5}
   }
   n41 = "Transform (Modify) Bone - Bone: neckbrace_up"(ComponentPose=@n40, Alpha=@n39, Translation="0,0,0", Rotation="0,0,0", Scale="1.5,1,1") {
    rotationMode=BMM_Additive
    rotationSpace=BCS_BoneSpace
    scaleMode=BMM_Additive
    scaleSpace=BCS_BoneSpace
    translationMode=BMM_Additive
    translationSpace=BCS_BoneSpace
    alphaBoolBlend {blendOption=Linear}
    alphaScaleBiasClamp {bMapRange=true, inRange {max=20}}
    boneToModify {boneName=neckbrace_up}
    scale {x=1.5}
   }
   n42 = "Transform (Modify) Bone - Bone: neckbrace_dn"(ComponentPose=@n41, Alpha=@n39, Translation="0,0,0", Rotation="0,0,0", Scale="1.5,1,1") {
    rotationMode=BMM_Additive
    rotationSpace=BCS_BoneSpace
    scaleMode=BMM_Additive
    scaleSpace=BCS_BoneSpace
    translationMode=BMM_Additive
    translationSpace=BCS_BoneSpace
    alphaBoolBlend {blendOption=Linear}
    alphaScaleBiasClamp {bMapRange=true, inRange {max=20}}
    boneToModify {boneName=neckbrace_dn}
    scale {x=1.5}
   }
   n43 = "Constraint - neckbrace_diamond (PPPPPP)"!(
    ComponentPose=@n42
    Alpha=1
    ConstraintWeights_0=0.15
    ConstraintWeights_1:float=0.1
    ConstraintWeights_2:float=0.1
    ConstraintWeights_3:float=0.25
    ConstraintWeights_4:float=0.25
    ConstraintWeights_5:float=0.15
   ) -> (Pose) {
    alphaBoolBlend {blendOption=Linear}
    boneToModify {boneName=neckbrace_diamond}
    constraintSetup [
     {targetBone {boneName=head}}
     {targetBone {boneName=neckbrace_dn}}
     {targetBone {boneName=neckbrace_up}}
     {targetBone {boneName=neckbrace_md}}
     {targetBone {boneName=spine_05}}
     {targetBone {boneName=neck_02}}
    ]
    constraintWeights [1,1,1,1,1,1]
   }
   n44 = "Transform (Modify) Bone - Bone: dyn_collar_fr_base_r"(ComponentPose=@n43, Alpha=Get uparmFwdBk_R() -> (uparmFwdBk_R:double), Translation="0,0,0", Rotation="40,40,0", Scale="1,1,1") {
    rotationMode=BMM_Additive
    rotationSpace=BCS_BoneSpace
    scaleMode=BMM_Additive
    scaleSpace=BCS_BoneSpace
    translationMode=BMM_Additive
    translationSpace=BCS_BoneSpace
    alphaBoolBlend {blendOption=Linear}
    alphaScaleBiasClamp {bMapRange=true, inRange {max=100, min=70}}
    boneToModify {boneName=dyn_collar_fr_base_r}
    rotation {pitch=40, yaw=40}
   }
   n45 = Get bIsSkydiveFloating()
   n46 = "Transform (Modify) Bone - Bone: dyn_collar_bk_base"!(ComponentPose=@n44, bAlphaBoolEnabled:bool=@n45, Translation="0,0,4", Rotation="-40,0,0", Scale="1,1,1") -> (Pose) {
    alphaInputType=Bool
    rotationMode=BMM_Additive
    rotationSpace=BCS_BoneSpace
    scaleMode=BMM_Additive
    scaleSpace=BCS_BoneSpace
    translationMode=BMM_Additive
    translationSpace=BCS_BoneSpace
    alphaBoolBlend {blendInTime=1, blendOption=Linear, blendOutTime=1}
    alphaScaleBiasClamp {inRange {max=100, min=70}}
    boneToModify {boneName=dyn_collar_bk_base}
    rotation {pitch=-40}
    translation {z=4}
   }
   n47 = "Transform (Modify) Bone - Bone: dyn_collar_side_base_l"!(ComponentPose=@n46, bAlphaBoolEnabled:bool=@n45, Translation="3,0,0", Rotation="-30,0,0", Scale="1,1,1") -> (Pose) {
    alphaInputType=Bool
    rotationMode=BMM_Additive
    rotationSpace=BCS_BoneSpace
    scaleMode=BMM_Additive
    scaleSpace=BCS_BoneSpace
    translationMode=BMM_Additive
    translationSpace=BCS_BoneSpace
    alphaBoolBlend {blendInTime=1, blendOption=Linear, blendOutTime=1}
    alphaScaleBiasClamp {inRange {max=100, min=70}}
    boneToModify {boneName=dyn_collar_side_base_l}
    rotation {pitch=-30}
    translation {x=3}
   }
   n48 = "Transform (Modify) Bone - Bone: dyn_collar_side_base_r"!(ComponentPose=@n47, bAlphaBoolEnabled:bool=@n45, Translation="-3,0,0", Rotation="-30,0,0", Scale="1,1,1") -> (Pose) {
    alphaInputType=Bool
    rotationMode=BMM_Additive
    rotationSpace=BCS_BoneSpace
    scaleMode=BMM_Additive
    scaleSpace=BCS_BoneSpace
    translationMode=BMM_Additive
    translationSpace=BCS_BoneSpace
    alphaBoolBlend {blendInTime=1, blendOption=Linear, blendOutTime=1}
    alphaScaleBiasClamp {inRange {max=100, min=70}}
    boneToModify {boneName=dyn_collar_side_base_r}
    rotation {pitch=-30}
    translation {x=-3}
   }
   n49 = Anim Dynamics - Bone: dyn_jacket_fr_b_l(ComponentPose=@n48, Alpha=1, ExternalForce="0,0,0") {
    angularDampingOverride=1
    angularSpringConstant=1500
    bAngularSpring=true
    bGravityOverrideInSimSpace=true
    bOverrideAngularDamping=true
    bUseSphericalLimits=true
    gravityScale=0.35
    simulationSpace=RootRelative
    alphaBoolBlend {blendOption=Linear}
    boundBone {boneName=dyn_jacket_fr_b_l}
    physicsBodyDefinitions [{
      collisionType=CustomSphere
      sphereCollisionRadius=3.5
      boundBone {boneName=dyn_jacket_fr_b_l}
      boxExtents=<10 20 20>
      constraintSetup {angularLimitsMax=<5 0 5>, angularLimitsMin=<-5 0 -15>, angularTarget=<0 0 0>}
      localJointOffset=<10 0 0>
     }]
    sphericalLimits [
     {limitRadius=8.5, drivingBone {boneName=spine_02}, sphereLocalOffset=<0 0 0>}
     {limitRadius=12, drivingBone {boneName=pelvis}, sphereLocalOffset=<0 0 0>}
    ]
   }
   n50 = Anim Dynamics - Bone: dyn_jacket_fr_b_r(ComponentPose=@n49, Alpha=1, ExternalForce="0,0,0") {
    angularDampingOverride=1
    angularSpringConstant=1500
    bAngularSpring=true
    bGravityOverrideInSimSpace=true
    bOverrideAngularDamping=true
    bUseSphericalLimits=true
    gravityScale=0.35
    simulationSpace=RootRelative
    alphaBoolBlend {blendOption=Linear}
    boundBone {boneName=dyn_jacket_fr_b_r}
    physicsBodyDefinitions [{
      collisionType=CustomSphere
      sphereCollisionRadius=3.5
      boundBone {boneName=dyn_jacket_fr_b_r}
      boxExtents=<10 20 20>
      constraintSetup {angularLimitsMax=<0 0 0>, angularLimitsMin=<0 0 -10>, angularTarget=<0 0 0>}
      localJointOffset=<-10 0 0>
     }]
    sphericalLimits [
     {limitRadius=8.5, drivingBone {boneName=spine_02}, sphereLocalOffset=<0 0 0>}
     {limitRadius=12, drivingBone {boneName=pelvis}, sphereLocalOffset=<0 0 0>}
    ]
   }
   n51 = Anim Dynamics - Bone: dyn_jacket_side_b_l(ComponentPose=@n50, Alpha=1, ExternalForce="0,0,0") {
    angularDampingOverride=1
    angularSpringConstant=1500
    bAngularSpring=true
    bGravityOverrideInSimSpace=true
    bOverrideAngularDamping=true
    bUseSphericalLimits=true
    gravityScale=0.35
    simulationSpace=RootRelative
    alphaBoolBlend {blendOption=Linear}
    boundBone {boneName=dyn_jacket_side_b_l}
    physicsBodyDefinitions [{
      collisionType=CustomSphere
      sphereCollisionRadius=3.5
      boundBone {boneName=dyn_jacket_side_b_l}
      boxExtents=<10 20 20>
      constraintSetup {angularLimitsMax=<5 5 5>, angularLimitsMin=<-5 -5 -10>, angularTarget=<0 0 0>}
      localJointOffset=<10 0 0>
     }]
    sphericalLimits [
     {limitRadius=10.5, drivingBone {boneName=spine_02}, sphereLocalOffset=<0 4.9611 0>}
     {limitRadius=12, drivingBone {boneName=pelvis}, sphereLocalOffset=<0 0 0>}
    ]
   }
   n52 = Anim Dynamics - Bone: dyn_jacket_side_b_r(ComponentPose=@n51, Alpha=1, ExternalForce="0,0,0") {
    angularDampingOverride=1
    angularSpringConstant=1500
    bAngularSpring=true
    bGravityOverrideInSimSpace=true
    bOverrideAngularDamping=true
    bUseSphericalLimits=true
    gravityScale=0.35
    simulationSpace=RootRelative
    alphaBoolBlend {blendOption=Linear}
    boundBone {boneName=dyn_jacket_side_b_r}
    physicsBodyDefinitions [{
      collisionType=CustomSphere
      sphereCollisionRadius=3.5
      boundBone {boneName=dyn_jacket_side_b_r}
      boxExtents=<10 20 20>
      constraintSetup {angularLimitsMax=<5 5 0>, angularLimitsMin=<-5 -5 -20>, angularTarget=<0 0 0>}
      localJointOffset=<-10 0 0>
     }]
    sphericalLimits [
     {limitRadius=12.5, drivingBone {boneName=spine_02}, sphereLocalOffset=<0 4.9611 0.5745>}
     {limitRadius=12, drivingBone {boneName=pelvis}, sphereLocalOffset=<0 0 0>}
    ]
   }
   n53 = Anim Dynamics - Bone: dyn_jacket_bk_b(ComponentPose=@n52, Alpha=1, ExternalForce="0,0,0") {
    angularDampingOverride=1
    angularSpringConstant=1500
    bAngularSpring=true
    bGravityOverrideInSimSpace=true
    bOverrideAngularDamping=true
    bUseSphericalLimits=true
    gravityScale=0.35
    simulationSpace=RootRelative
    alphaBoolBlend {blendOption=Linear}
    boundBone {boneName=dyn_jacket_bk_b}
    physicsBodyDefinitions [{
      collisionType=CustomSphere
      sphereCollisionRadius=3.5
      boundBone {boneName=dyn_jacket_bk_b}
      boxExtents=<10 20 20>
      constraintSetup {angularLimitsMax=<0 0 0>, angularLimitsMin=<0 0 -15>, angularTarget=<0 0 0>}
      localJointOffset=<10 0 0>
     }]
    sphericalLimits [
     {limitRadius=9.5, drivingBone {boneName=spine_02}, sphereLocalOffset=<0 1.5418 0>}
     {limitRadius=12, drivingBone {boneName=pelvis}, sphereLocalOffset=<0 0 0>}
    ]
   }
   n54 = Anim Dynamics - Bone: dyn_jacket_zipper(ComponentPose=@n53, Alpha=1, ExternalForce="0,0,0") {
    angularDampingOverride=1
    angularSpringConstant=1500
    bAngularSpring=true
    bGravityOverrideInSimSpace=true
    bOverrideAngularDamping=true
    gravityScale=0.35
    simulationSpace=RootRelative
    alphaBoolBlend {blendOption=Linear}
    boundBone {boneName=dyn_jacket_zipper}
    physicsBodyDefinitions [{
      sphereCollisionRadius=0
      boundBone {boneName=dyn_jacket_zipper}
      boxExtents=<10 20 20>
      constraintSetup {angularLimitsMax=<0 0 25>, angularLimitsMin=<0 0 0>, angularTarget=<0 0 0>}
      localJointOffset=<5 0 0>
     }]
   }
   n55 = Anim Dynamics - Bone: dyn_chest_zipper(ComponentPose=@n54, Alpha=1, ExternalForce="0,0,0") {
    angularDampingOverride=1
    angularSpringConstant=1500
    bAngularSpring=true
    bGravityOverrideInSimSpace=true
    bOverrideAngularDamping=true
    gravityScale=0.35
    simulationSpace=RootRelative
    alphaBoolBlend {blendOption=Linear}
    boundBone {boneName=dyn_chest_zipper}
    physicsBodyDefinitions [{
      sphereCollisionRadius=0
      boundBone {boneName=dyn_chest_zipper}
      boxExtents=<10 20 20>
      constraintSetup {angularLimitsMax=<0 0 45>, angularLimitsMin=<0 0 0>, angularTarget=<0 0 0>}
      localJointOffset=<5 0 0>
     }]
   }
   n56 = Anim Dynamics - Bone: dyn_gun(ComponentPose=@n55, Alpha=0.5, ExternalForce="0,0,0") {
    angularDampingOverride=1
    angularSpringConstant=1500
    bAngularSpring=true
    bGravityOverrideInSimSpace=true
    bOverrideAngularDamping=true
    gravityScale=0.35
    simulationSpace=RootRelative
    alphaBoolBlend {blendOption=Linear}
    boundBone {boneName=dyn_gun}
    physicsBodyDefinitions [{
      sphereCollisionRadius=0
      boundBone {boneName=dyn_gun}
      boxExtents=<10 10 10>
      constraintSetup {angularLimitsMax=<0 3 3>, angularLimitsMin=<0 -3 -3>, angularTarget=<0 0 0>}
      localJointOffset=<2 0 0>
     }]
   }
   n57 = RigidBody(ComponentPose=@n56) {
    bClampLinearTranslationLimitToRefPose=true
    bOverrideWorldGravity=true
    overridePhysicsAsset=&/F_MED_Cotton_Candy_Physics
    simulationSpace=BaseBoneSpace
    alphaBoolBlend {blendOption=Linear}
    baseBoneRef {boneName=pelvis}
    componentLinearAccScale=<0.1 0.1 0.1>
    componentLinearVelScale=<0.1 0.1 0.1>
    overrideWorldGravity {z=-980}
   }
   n58 = Control Rig!(Source=Component To Local(ComponentPose=@n57), Alpha, NoAnimDynamics:bool=Get AnimDynAlpha() -> (AnimDynAlpha:bool)) -> (Pose) {
    alphaBoolBlend {blendOption=Linear}
    controlRigAssetReference {blueprintRigClass=&/F_MED_Cotton_Candy_CtrlRig.F_MED_Cotton_Candy_CtrlRig_C}
   }
   Output Pose(Result=@n58) {layerGroup=None}
  ]}
 EventGraph {nodes [
   Event BlueprintUpdateAnimation()
   > n1 = Sequence!(execute) -> (then_0, then_1, then_2:Exec, then_3:Exec, then_4:Exec, then_5:Exec, then_6:Exec, then_7:Exec, then_8:Exec, then_9:Exec)
   n2 = Branch(execute=@n1.then_9, Condition=AND Boolean(A=Get bAnimDynamicsEnabled() -> (bAnimDynamicsEnabled:bool), B=Get bRigidBodyEnabled()))
   > Set AnimDynAlpha(AnimDynAlpha:bool=false) -> (Output_Get:bool)
   Set AnimDynAlpha(execute=@n2.else, AnimDynAlpha=true)
   n3 = GetDeltaTransformFromRefPose(self=GetOwnerSkeletalMeshComponent(), BoneName=lowerarm_l, BaseName=upperarm_l)
   Set elbow_rot_l(execute=@n1.then_0, elbow_rot_l:double=BreakRotator(InRot=BreakTransform(InTransform=@n3).Rotation).Yaw) -> (Output_Get:double)
   n4 = GetDeltaTransformFromRefPose(self=GetOwnerSkeletalMeshComponent(), BoneName=thigh_r, BaseName=pelvis)
   Set thigh_rot_r(execute=@n1.then_3, thigh_rot_r=BreakRotator(InRot=BreakTransform(InTransform=@n4).Rotation).Yaw)
   n5 = GetDeltaTransformFromRefPose(self=GetOwnerSkeletalMeshComponent(), BoneName=calf_r, BaseName=thigh_r)
   Set knee_rot_r(execute=@n1.then_5, knee_rot_r:double=BreakRotator(InRot=BreakTransform(InTransform=@n5).Rotation).Yaw) -> (Output_Get:double)
   n6 = GetDeltaTransformFromRefPose(self=GetOwnerSkeletalMeshComponent(), BoneName=calf_l, BaseName=thigh_l)
   Set knee_rot_l(execute=@n1.then_4, knee_rot_l:double=BreakRotator(InRot=BreakTransform(InTransform=@n6).Rotation).Yaw) -> (Output_Get:double)
   n7 = GetDeltaTransformFromRefPose(self=GetOwnerSkeletalMeshComponent(), BoneName=head, BaseName=neck_01)
   Set HeadUpDn(execute=@n1.then_8, HeadUpDn:double=BreakRotator(InRot=BreakTransform(InTransform=@n7).Rotation).Yaw) -> (Output_Get:double)
   n8 = GetDeltaTransformFromRefPose(self=GetOwnerSkeletalMeshComponent(), BoneName=upperarm_l, BaseName=clavicle_l)
   Set uparmFwdBk_L(execute=@n1.then_6, uparmFwdBk_L:double=BreakRotator(InRot=BreakTransform(InTransform=@n8).Rotation).Yaw) -> (Output_Get:double)
   n9 = GetDeltaTransformFromRefPose(self=GetOwnerSkeletalMeshComponent(), BoneName=thigh_l, BaseName=pelvis)
   Set thigh_rot_l(execute=@n1.then_2, thigh_rot_l=BreakRotator(InRot=BreakTransform(InTransform=@n9).Rotation).Yaw)
   n10 = GetDeltaTransformFromRefPose(self=GetOwnerSkeletalMeshComponent(), BoneName=lowerarm_r, BaseName=upperarm_r)
   Set elbow_rot_r(execute=@n1.then_1, elbow_rot_r:double=BreakRotator(InRot=BreakTransform(InTransform=@n10).Rotation).Yaw) -> (Output_Get:double)
   n11 = GetDeltaTransformFromRefPose(self=GetOwnerSkeletalMeshComponent(), BoneName=upperarm_r, BaseName=clavicle_r)
   Set uparmFwdBk_R(execute=@n1.then_7, uparmFwdBk_R:double=BreakRotator(InRot=BreakTransform(InTransform=@n11).Rotation).Yaw) -> (Output_Get:double)
   TryGetPawnOwner()
  ]}
]
variables [
 AnimDynAlpha {type=boolean}
 elbow_rot_l {type=number}
 elbow_rot_r {type=number}
 HeadUpDn {type=number}
 knee_rot_l {type=number}
 knee_rot_r {type=number}
 thigh_rot_l {type=number}
 thigh_rot_r {type=number}
 uparmFwdBk_L {type=number}
 uparmFwdBk_R {type=number}
]
